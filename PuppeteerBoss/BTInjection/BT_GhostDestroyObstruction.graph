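# Behavior tree for a Puppeteer boss ghost that walks toward its obstruction target, apparently injected to have it destroy obstacles.
# A root selector runs Bot Move To on the GhostTargetActor blackboard key using the Phoebe nav filter, falling back to a short Wait if that fails.
# Services on the selector handle clamber execution, weapon management and an NPC melee attack evaluator, with no explicit attack task in the tree.
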BehaviorTree /PuppeteerBoss/BTInjection/BT_GhostDestroyObstruction
blackboard=&/BB_GhostDestroyObstacle
root {
 name=Selector
 class=BTComposite_Selector
 children [
  Bot Move To {
   class=FortAthenaBTTask_BotMoveTo
   decorators []
   properties {
    movementResultKeyName=None
    nodeName=Bot Move To
    acceptableRadius {defaultValue=5}
    allowRandomWobble {defaultValue=true}
    bAllowStrafe {defaultValue=false}
    blackboardKey {selectedKeyName=GhostTargetActor}
    filterClass {defaultValue=/Game/Athena/AI/Phoebe/NavFilter/NavFilter_Phoebe.NavFilter_Phoebe_C}
    observedBlackboardValueTolerance {defaultValue=4.75}
   }
  }
  Wait {class=BTTask_Wait, properties {waitTime {defaultValue=1}}}
 ]
 decorators []
 properties {
  children [
   {childComposite=null, childTask=&self:FortAthenaBTTask_BotMoveTo_0, decorators []}
   {childComposite=null, childTask=&self:BTTask_Wait_0, decorators []}
  ]
  services [
   &self:FortAthenaBTService_Clamber_0
   &self:FortAthenaBTService_ManageWeapon_0
   &self:FortAthenaBTService_AIEvaluator_0
  ]
 }
 services [
  Executing... Clamber {
   class=FortAthenaBTService_Clamber
   properties {
    bCallTickOnSearchStart=false
    bRestartTimerOnEachActivation=false
    clamberAbilityStatusName=AIEvaluator_Clamber_AbilityStatus
    clamberAttemptRetryMaxCount=2
    clamberDestinationLocationName=AIEvaluator_Clamber_Destination
    clamberExecutionStatusName=AIEvaluator_Clamber_ExecutionStatus
    clamberOriginLocationName=AIEvaluator_Clamber_Origin
    clamberRetryDelay=0.15
    crouchExecutionStatusName=AIEvaluator_Crouch_ExecutionStatus
    firstJumpClamberMaxStartDelay=2
    firstJumpRetryDelay=0.1
    firstJumpRetryMaxCount=5
    interval=0.5
    jumpExecutionStatusName=AIEvaluator_Jump_ExecutionStatus
    nodeName=Executing... Clamber
    randomDeviation=0.1
   }
  }
  AI Bot Manage Weapon {class=FortAthenaBTService_ManageWeapon, properties {interval=0.5, nodeName=AI Bot Manage Weapon, randomDeviation=0.1}}
  AI Evaluator {
   class=FortAthenaBTService_AIEvaluator
   properties {
    interval=0.5
    nodeName=AI Evaluator
    aIEvaluatorClass {defaultValue=/Game/Athena/AI/NPCs/Base/Evaluators/Evaluator_MeleeAttack_NPC.Evaluator_MeleeAttack_NPC_C}
   }
  }
 ]
}
