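# BT_SeenWait is a behavior tree on the VerseNPC blackboard BB_Vk_NPC, in the PuppeteerBoss BT injection folder, that holds an NPC idle.
# A root Selector runs a single Wait task with an extremely long duration, so the tree effectively waits indefinitely with no decorators or services.
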BehaviorTree /PuppeteerBoss/BTInjection/BT_SeenWait
blackboard=/VerseNPC/BehaviorTrees/BB_Vk_NPC
root {
 name=Selector
 class=BTComposite_Selector
 children [Wait {class=BTTask_Wait, properties {waitTime {defaultValue=10000000}}}]
 decorators []
 properties {children [{childComposite=null, childTask=&self:BTTask_Wait_0, decorators []}], services []}
 services []
}
