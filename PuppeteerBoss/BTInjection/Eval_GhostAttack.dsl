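; Behavior tree injection evaluator for the Puppeteer boss's ghost AI that gates its attack on players being nearby.
; On enter it caches the ghost controller and pawn, then registers a nearby-actors perception listener using the attack minimum distance.
; Listener callbacks update a nearby-players flag and request re-evaluation, so the starting condition simply returns that flag.
; On exit, if the perception component is valid, it unregisters the listener and clears the flag.
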(event EventOnEnter
  (params (OwnerComp :type Behavior%20Tree%20Component%20Object%20Reference))
  (bind _ascontroller (Utilities|Casting|CastToController (Components|GetOwner OwnerComp)))
  (bind _asfort_athena_aibot_controller (Utilities|Casting|CastToFortAthenaAIBotController _ascontroller))
  (bind _output_get (Variables|Default|SetGhostController _asfort_athena_aibot_controller))
  (bind _asbp_player_pawn_puppeteer_ghost (Utilities|Casting|CastToBP_PlayerPawn_Puppeteer_Ghost (Pawn|GetControlledPawn _output_get)))
  (bind _output_get_2 (Variables|Default|SetGhostPawn _asbp_player_pawn_puppeteer_ghost))
  (bind _returnvalue (Query|FindorCreateNearbyActorsPerceptionComponentonActor _output_get_2 "/Script/FortniteAI.FortAINearbyActorsPerceptionComponent"))
  (bind _output_get_3 (Variables|Default|SetNearbyActorComponent _returnvalue))
  (bind _returnvalue_2 (Query|BPRegisterListener _output_get_3 (Utilities|Struct|MakeFortNearbyActorsPerceptionConfiguration 0.5 3 0 (ScalableFloat|GetValueatLevel (Variables|Default|GetAttackMinDistance_Hf))) (delegate Custom|OnNearbyPlayerPerception)))
  (Variables|Default|SetNearbyActorHandle _returnvalue_2))

(event Custom|OnNearbyPlayerPerception
  (params (AllFoundActors :type Array%20of%20Actor%20Object%20References) (EnteringActors :type Array%20of%20Actor%20Object%20References) (ExitingActors :type Array%20of%20Actor%20Object%20References))
  (Variables|Default|SetHasNearbyPlayers (> (Utilities|Array|Length AllFoundActors) 0))
  (CallFunction|RequestEvaluation))

(event EventOnExit
  (params (OwnerComp :type Behavior%20Tree%20Component%20Object%20Reference))
  (Utilities|IsValid (Variables|Default|GetNearbyActorComponent)
    (:Is%20Valid
      (Query|BPUnregisterListener (Variables|Default|GetNearbyActorHandle))
      (Variables|Default|SetHasNearbyPlayers false))))

(fn EvaluateStartingConditions
  (params (OwnerComp :type Behavior%20Tree%20Component%20Object%20Reference))
  (return (Variables|Default|GetHasNearbyPlayers)))
