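; Behavior tree injection evaluator that passes when the Puppeteer boss ghost is being seen by a nearby player.
; On enter it registers a nearby-actors perception listener that caches player pawns, then unregisters on exit.
; With a target player and no attack underway, it checks each player's camera yaw against a maximum angle, then line-traces to the ghost.
; The result adds or removes the ghost's loose Seen gameplay tag.
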(event EventOnEnter
  (params (OwnerComp :type Behavior%20Tree%20Component%20Object%20Reference))
  (bind _ascontroller (Utilities|Casting|CastToController (Components|GetOwner OwnerComp)))
  (bind _asfort_athena_aibot_controller (Utilities|Casting|CastToFortAthenaAIBotController _ascontroller))
  (bind _output_get (Variables|Default|SetGhostController _asfort_athena_aibot_controller))
  (bind _asbp_player_pawn_puppeteer_ghost (Utilities|Casting|CastToBP_PlayerPawn_Puppeteer_Ghost (Pawn|GetControlledPawn _output_get)))
  (bind _output_get_2 (Variables|Default|SetGhostPawn _asbp_player_pawn_puppeteer_ghost))
  (bind _returnvalue (Query|FindorCreateNearbyActorsPerceptionComponentonActor _output_get_2 "/Script/FortniteAI.FortAINearbyActorsPerceptionComponent"))
  (bind _output_get_3 (Variables|Default|SetNearbyActorComponent _returnvalue))
  (bind _returnvalue_2 (Query|BPRegisterListener _output_get_3 (Utilities|Struct|MakeFortNearbyActorsPerceptionConfiguration 1.0 3 0 (ScalableFloat|GetValueatLevel (Variables|Default|GetLosPlayerCheckRadius_Hf))) (delegate Custom|OnNearbyActorChanged)))
  (Variables|Default|SetListenerHandle _returnvalue_2))

(event EventOnExit
  (params (OwnerComp :type Behavior%20Tree%20Component%20Object%20Reference))
  (Utilities|IsValid (Variables|Default|GetNearbyActorComponent)
    (:Is%20Valid
      (Query|BPUnregisterListener (Variables|Default|GetListenerHandle)))))

(event Custom|OnNearbyActorChanged
  (params (AllFoundActors :type Array%20of%20Actor%20Object%20References) (EnteringActors :type Array%20of%20Actor%20Object%20References) (ExitingActors :type Array%20of%20Actor%20Object%20References))
  (bind _filteredarray (Utilities|Array|FilterArray AllFoundActors "/Script/FortniteGame.FortPlayerPawnAthena"))
  (Variables|Default|SetNearbyPlayerPawns _filteredarray))

(fn CanBeSeen
  (bind _returnvalue (GameplayTags|MakeGameplayTagContainerfromTag "(TagName=\"PuppeteerBoss.Ghost.Seen\")"))
  (bind _ghostpawn (Variables|Default|GetGhostPawn))
  (bind _returnvalue_2 (GameplayTags|HasMatchingGameplayTag _ghostpawn "(TagName=\"PuppeteerBoss.Ghost.Seen\")"))
  (bind _ghostpawn_2 (Variables|Default|GetGhostPawn))
  (Utilities|IsValid _ghostpawn_2
    (:Is%20Valid
      (Utilities|IsValid (Variables|/PuppeteerBoss/Pawn/BP_PlayerPawn_Puppeteer_Ghost.BP_PlayerPawn_Puppeteer_Ghost_C|GetTargetPlayer _ghostpawn_2)
        (:Is%20Valid
          (if (GameplayTags|HasMatchingGameplayTag (Variables|Default|GetGhostPawn) "(TagName=\"PuppeteerBoss.Ghost.IsAttacking\")")
            (return false)
            (else
              (LocalVariables|SetLosCheckLoopBroke false)
              (bind _array_element (Utilities|Array|ForEachLoopwithBreak (Variables|Default|GetNearbyPlayerPawns))
                (:LoopBody
                  (bind (_location _rotation) (CallFunction|GetPlayerPawnCameraInfo _array_element))
                  (Utilities|IsValid _array_element
                    (:Is%20Valid
                      (bind (_returnvalue_roll _returnvalue_pitch _returnvalue_yaw) (Math|Rotator|BreakRotator (Math|Rotator|Delta(Rotator) _rotation (Math|Rotator|FindLookatRotation _location (Transformation|GetActorLocation (Variables|Default|GetGhostPawn))))))
                      (if (<= (Math|Float|Absolute(Float) _returnvalue_yaw) (ScalableFloat|GetValueatLevel (Variables|Default|GetSeenMaxAngle_Hf)))
                        (bind _outhit (Collision|LineTraceByChannel _location (Transformation|GetActorLocation (Variables|Default|GetGhostPawn)) "TraceTypeQuery1" false (Utilities|Array|MakeArray _array_element)))
                        (if (== (.HitActor (Collision|BreakHitResult _outhit)) (Variables|Default|GetGhostPawn))
                          (LocalVariables|SetLosCheckLoopBroke true))))))
                (:Completed
                  _))
              (if (LocalVariables|GetLosCheckLoopBroke)
                (if (not _returnvalue_2)
                  (Ability|GameplayEffect|AddLooseGameplayTags (Variables|Default|GetGhostPawn) _returnvalue))
                (return true)
                (elif _returnvalue_2
                  (Ability|GameplayEffect|RemoveLooseGameplayTags (Variables|Default|GetGhostPawn) _returnvalue)
                  (return false)
                  (else
                    _))))))
        (:Is%20Not%20Valid
          (return false))))
    (:Is%20Not%20Valid
      (return false))))

(fn EvaluateStartingConditions
  (params (OwnerComp :type Behavior%20Tree%20Component%20Object%20Reference))
  (bind _seen (CallFunction|CanbeSeen))
  (return _seen))

(fn GetPlayerPawnCameraInfo
  (params (PlayerPawn :type Fort%20Player%20Pawn%20Athena%20Object%20Reference))
  (bind _returnvalue (Fortnite|GetFortPlayerControllerfromActor PlayerPawn))
  (Utilities|IsValid _returnvalue
    (:Is%20Valid
      (return (Camera|GetCameraLocation (Variables|/Script/Engine.PlayerController|GetPlayerCameraManager _returnvalue)) (Camera|GetCameraRotation (Variables|/Script/Engine.PlayerController|GetPlayerCameraManager _returnvalue))))
    (:Is%20Not%20Valid
      (return (Transformation|GetActorLocation PlayerPawn) (Transformation|GetActorRotation PlayerPawn)))))
