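; Behavior tree evaluator for the Puppeteer boss that checks whether its ghost target exists but cannot be reached.
; It overrides EvaluateStartingConditions, reading a target-reachable bool and a ghost target actor from the blackboard through key selector variables.
; The condition passes only when the reachable flag is false and the ghost target actor is valid.
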(fn EvaluateStartingConditions
  (params (OwnerComp :type Behavior%20Tree%20Component%20Object%20Reference))
  (return (and (not (Blackboard|GetBool (Variables|Default|GetIsTargetReachable) OwnerComp)) (Utilities|IsValid (Blackboard|GetObject (Variables|Default|GetGhostTargetActor) OwnerComp) :!class /Script/Engine.KismetSystemLibrary))))
