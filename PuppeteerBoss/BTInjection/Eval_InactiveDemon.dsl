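; Behavior tree injection evaluator that decides when a Puppeteer boss demon's inactive behavior can start and keep running.
; On enter it caches the AI bot controller and its demon pawn.
; Starting requires a valid, unstaggered mask and an active Phase 2 demon that is neither in nor teleporting to the spawn cloud.
; Continuing requires the demon's inactive flag to be set, with the same spawn-cloud exclusions.
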(event EventOnEnter
  (params (OwnerComp :type Behavior%20Tree%20Component%20Object%20Reference))
  (bind _ascontroller (Utilities|Casting|CastToController (Components|GetOwner OwnerComp)))
  (bind _asfort_athena_aibot_controller (Utilities|Casting|CastToFortAthenaAIBotController _ascontroller))
  (bind _output_get (Variables|Default|SetDemonController _asfort_athena_aibot_controller))
  (bind _asbp_player_pawn_puppeteer_demon (Utilities|Casting|CastToBP_PlayerPawn_Puppeteer_Demon (Pawn|GetControlledPawn _output_get)))
  (Variables|Default|SetDemonPawn _asbp_player_pawn_puppeteer_demon))

(fn EvaluateOngoingConditions
  (params (OwnerComp :type Behavior%20Tree%20Component%20Object%20Reference))
  (return (Math|Boolean|ANDBoolean (Variables|/PuppeteerBoss/Pawn/BP_PlayerPawn_Puppeteer_Demon.BP_PlayerPawn_Puppeteer_Demon_C|GetIsInactive (Variables|Default|GetDemonPawn)) (not (GameplayTags|HasMatchingGameplayTag (Variables|Default|GetDemonPawn) "(TagName=\"FirePetalBosses.Demons.IsTeleportingToSpawnCloud\")")) (not (GameplayTags|HasMatchingGameplayTag (Variables|Default|GetDemonPawn) "(TagName=\"FirePetalBosses.Demons.IsinSpawnCloud\")")))))

(fn EvaluateStartingConditions
  (params (OwnerComp :type Behavior%20Tree%20Component%20Object%20Reference))
  (bind _demonpawn (Variables|Default|GetDemonPawn))
  (Utilities|IsValid (Variables|/PuppeteerBoss/Pawn/BP_PlayerPawn_Puppeteer_Demon.BP_PlayerPawn_Puppeteer_Demon_C|GetMaskActor _demonpawn)
    (:Is%20Valid
      (return (Math|Boolean|ANDBoolean (not (Variables|/PuppeteerBoss/Pawn/BP_PlayerPawn_Puppeteer_Demon.BP_PlayerPawn_Puppeteer_Demon_C|GetMaskStaggered _demonpawn)) (not (GameplayTags|HasMatchingGameplayTag _demonpawn "(TagName=\"FirePetalBosses.Demons.IsTeleportingToSpawnCloud\")")) (GameplayTags|HasMatchingGameplayTag _demonpawn "(TagName=\"PuppeteerBoss.Demon.IsInPhase2\")") (not (GameplayTags|HasMatchingGameplayTag _demonpawn "(TagName=\"FirePetalBosses.Demons.IsinSpawnCloud\")")) (not (Variables|/PuppeteerBoss/Pawn/BP_PlayerPawn_Puppeteer_Demon.BP_PlayerPawn_Puppeteer_Demon_C|GetIsInactive _demonpawn)) (not (GameplayTags|HasMatchingGameplayTag (Variables|/PuppeteerBoss/Pawn/BP_PlayerPawn_Puppeteer_Demon.BP_PlayerPawn_Puppeteer_Demon_C|GetMaskActor _demonpawn) "(TagName=\"PuppeteerBoss.Mask.IsStaggered\")")))))
    (:Is%20Not%20Valid
      (return false))))
