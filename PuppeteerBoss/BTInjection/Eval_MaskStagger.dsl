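; Behavior tree evaluator for the Puppeteer boss demon that decides whether its mask-stagger behavior can start.
; On enter, it casts the component's owner to the Athena AI bot controller and caches it along with the controlled demon pawn.
; The start condition passes only while the pawn is flagged mask-staggered and its stagger animation has not yet started.
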(event EventOnEnter
  (params (OwnerComp :type Behavior%20Tree%20Component%20Object%20Reference))
  (bind _ascontroller (Utilities|Casting|CastToController (Components|GetOwner OwnerComp)))
  (bind _asfort_athena_aibot_controller (Utilities|Casting|CastToFortAthenaAIBotController _ascontroller))
  (bind _output_get (Variables|Default|SetDemonController _asfort_athena_aibot_controller))
  (bind _asbp_player_pawn_puppeteer_demon (Utilities|Casting|CastToBP_PlayerPawn_Puppeteer_Demon (Pawn|GetControlledPawn _output_get)))
  (Variables|Default|SetDemonPawn _asbp_player_pawn_puppeteer_demon))

(fn EvaluateStartingConditions
  (params (OwnerComp :type Behavior%20Tree%20Component%20Object%20Reference))
  (return (and (Variables|/PuppeteerBoss/Pawn/BP_PlayerPawn_Puppeteer_Demon.BP_PlayerPawn_Puppeteer_Demon_C|GetMaskStaggered (Variables|Default|GetDemonPawn)) (not (Variables|/PuppeteerBoss/Pawn/BP_PlayerPawn_Puppeteer_Demon.BP_PlayerPawn_Puppeteer_Demon_C|GetMaskStaggerAnimStarted (Variables|Default|GetDemonPawn))))))
